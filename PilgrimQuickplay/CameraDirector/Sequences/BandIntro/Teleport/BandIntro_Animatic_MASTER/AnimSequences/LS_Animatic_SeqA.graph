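# Level Sequence for sequence A of the animatic in the Pilgrim Quickplay band intro, under the Teleport variant of the camera director sequences.
# A single cinematic shot track named Shots holds three back-to-back shot sections that form one continuous cut.
# It has no bindings or other tracks, and the sections show no referenced shot sequences, so only the shot timing is defined here.
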LevelSequence /PilgrimQuickplay/CameraDirector/Sequences/BandIntro/Teleport/BandIntro_Animatic_MASTER/AnimSequences/LS_Animatic_SeqA
bindings []
display_rate {denominator=1, numerator=30}
tick_resolution {denominator=1, numerator=24000}
tracks [{
  type=MovieSceneCinematicShotTrack
  display_name=Shots
  sections [
   {type=MovieSceneCinematicShotSection, end_frame=126, start_frame=0}
   {type=MovieSceneCinematicShotSection, end_frame=288, start_frame=201}
   {type=MovieSceneCinematicShotSection, end_frame=201, start_frame=126}
  ]
 }]
