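# Level sequence for a portrait variant of animatic Sequence A in the Pilgrim Quickplay band intro teleport camera sequences.
# It holds three cinematic shot tracks with one section each, at a 30 fps display rate.
# The first two shots in time run back to back, and the third starts before the second ends, overlapping it.
# It has no object bindings, and no shot sequences are shown on the sections.
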LevelSequence /PilgrimQuickplay/CameraDirector/Sequences/BandIntro/Teleport/BandIntro_Animatic_MASTER/AnimSequences/LS_Animatic_SeqA_Portrait
bindings []
display_rate {denominator=1, numerator=30}
tick_resolution {denominator=1, numerator=24000}
tracks [
 {type=MovieSceneCinematicShotTrack, display_name=Shots, sections [{type=MovieSceneCinematicShotSection, end_frame=267, start_frame=61}]}
 {type=MovieSceneCinematicShotTrack, display_name=Shots, sections [{type=MovieSceneCinematicShotSection, end_frame=61, start_frame=0}]}
 {type=MovieSceneCinematicShotTrack, display_name=Shots, sections [{type=MovieSceneCinematicShotSection, end_frame=477, start_frame=163}]}
]
